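# Niagara system that draws summon link beams between a Spire and the spike positions passed in as a user array, one particle per spike.
# The Spline emitter places a mesh at each spike, orienting and stretching it toward the emitter origin, with distance-based alpha.
# Timeline and curves drive its material parameters, fades and brightness pulse.
# A second emitter adds a pink, camera-offset sprite flash at each spike that fades in and out.
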NiagaraSystem /SpireSharedAssets/VFX/Summon/NS_Spire_Summon_Link
dependencies {
 usedDataInterfaces [&NiagaraDataInterfaceCurve, &NiagaraDataInterfaceSkeletalMesh]
 usedDynamicInputs []
 usedModules [
  &self:SystemSpawnScript.SystemScriptSource.SystemScriptGraph.NiagaraNodeAssignment_0.SetVariables_#1
  &SystemState
 ]
}
emitters {
 Minimal004 {
  emitterUpdateScript [
   &EmitterState()
   &SpawnBurst_Instantaneous(Spawn Count=$User.SpikeAmount, Use Spawn Probability=false, UseLoopCountLimit=false)
  ]
  particleSpawnScript [
   &InitializeParticle(
    Color=<rgba 1 0 0.1667 1>
    Lifetime=$System.CurrentLoopDuration
    Sprite Size=<1300 350>
    Sprite Size Mode=::NewEnumerator4
    UsePositionOffset=false
    Write Material Random=true
   )
   &CameraOffset(Camera Offset Amount=500)
  ]
  particleUpdateScript [
   &ParticleState()
   /SetVariables(Particles.Position=&SelectVectorFromArray(Direct Array Index=&ReturnExecIndex(), Vector Selection Array=$User.SpikePositionArray)) {isSetParameters=true}
   &ScaleColor(
    Scale Alpha=&FloatFromCurve(FloatCurve=DataInterface() {propertyValues {Curve {
        defaultValue=0
        keys [
         {arriveTangent=9.0725, interpMode=RCIM_Cubic, leaveTangent=9.0725, tangentMode=RCTM_User, time=0, value=0}
         {arriveTangent=-2.5249, interpMode=RCIM_Cubic, leaveTangent=-2.5249, tangentMode=RCTM_User, time=0.2, value=0.9847}
         {arriveTangent=-0.0405, interpMode=RCIM_Cubic, leaveTangent=-0.0405, tangentMode=RCTM_User, time=1, value=0}
        ]
       }}})
    ScaleA=true
    ScaleRGB=true
   )
  ]
  rendererClasses [&NiagaraSpriteRendererProperties]
  renderers [{rendererClass=&NiagaraSpriteRendererProperties}]
 }
 Spline {
  emitterUpdateScript [
   &EmitterState()
   &SpawnBurst_Instantaneous(Spawn Count=$User.SpikeAmount, Use Spawn Probability=false, UseLoopCountLimit=false)
  ]
  particleSpawnScript [
   &InitializeParticle(
    Color=<rgba 1 0.03 0.1917 0.2>
    Lifetime=$System.CurrentLoopDuration
    Mesh Scale Mode=::NewEnumerator3
    UsePositionOffset=false
    Write Material Random=true
   )
   &CameraOffset(Camera Offset Amount=10)
  ]
  particleUpdateScript [
   &ParticleState()
   /SetVariables(Particles.Position=&SelectVectorFromArray(Direct Array Index=&ReturnExecIndex(), Vector Selection Array=$User.SpikePositionArray)) {isSetParameters=true}
   &ScaleMeshSize(Scale Factor=&MakeVector(
     X=1
     Y=1
     Z=&Multiply_Float(
      A=&Divide_Float(
       A=&VectorLength(Vector=&Subtract_Vector(
         A=&SelectVectorFromArray(Direct Array Index=&ReturnExecIndex(), Vector Selection Array=$User.SpikePositionArray)
         B=Linked() {linkedVariable {name=Engine.Emitter.SimulationPosition, type=NiagaraPosition}}
        ))
       B=&Multiply_Float(A=10, B=1)
      )
      B=&FloatFromCurve(FloatCurve=DataInterface() {propertyValues {Curve {keys [<at 0 1>, {arriveTangent=0.1406, time=0.2, value=1}]}}})
     )
    ))
   /SetVariables(Particles.MeshOrientation=&FindQuatBetweenTwoVectors(
     From Vector=<0 0 1>
     To Vector=&NormalizeVector(
      Vector To Normalize=&Subtract_Vector(
       A=Linked() {linkedVariable {name=Engine.Emitter.SimulationPosition, type=NiagaraPosition}}
       B=&SelectVectorFromArray(Direct Array Index=&ReturnExecIndex(), Vector Selection Array=$User.SpikePositionArray)
      )
      Safe Normalize=true
      Safe Fallback Threshold=0
      Safe Fallback Vector=<0 0 0.001>
     )
    )) {isSetParameters=true}
   &ScaleColor(
    Scale Alpha=&Divide_Float(
     A=1183.0546
     B=&VectorLength(Vector=&Subtract_Vector(
       A=&SelectVectorFromArray(Direct Array Index=&ReturnExecIndex(), Vector Selection Array=$User.SpikePositionArray)
       B=Linked() {linkedVariable {name=Engine.Emitter.SimulationPosition, type=NiagaraPosition}}
      ))
    )
    ScaleA=true
    ScaleRGB=true
   )
   &Timeline(Play Rate=-1.5, Rewind Mode:ENiagara_TimelineMode=::NewEnumerator1, Timeline Beginning=-999999, Timeline End=9999999)
   &DynamicMaterialParameters(
    Index 0 Param 1=$Particles.Timeline.Time
    Index 0 Param 2=&FloatFromCurve(FloatCurve=DataInterface() {propertyValues {Curve {keys [
         {interpMode=RCIM_Cubic, leaveTangent=7.954, tangentMode=RCTM_Break, time=0, value=0}
         {interpMode=RCIM_Cubic, tangentMode=RCTM_Break, time=0.3, value=1}
         <at 1 1>
        ]}}})
    Module.Param0WriteEnabled=true
    Module.Param1WriteEnabled=true
    Module.Param2WriteEnabled=true
    Module.Param3WriteEnabled=true
   )
   &ScaleColor(
    Scale Alpha=&FloatFromCurve(FloatCurve=DataInterface() {propertyValues {Curve {keys [
         {interpMode=RCIM_Cubic, tangentMode=RCTM_Break, time=0, value=0}
         {arriveTangent=0.9798, interpMode=RCIM_Cubic, leaveTangent=-3.4063, tangentMode=RCTM_Break, time=0.3, value=1}
         {arriveTangent=-3.0955, tangentMode=RCTM_User, time=1, value=0}
        ]}}})
    Scale RGB=&VectorFromFloat(Value=&FloatFromCurve(
      FloatCurve=DataInterface() {propertyValues {Curve {keys [
          {interpMode=RCIM_Cubic, tangentMode=RCTM_Break, time=0, value=0.7}
          {interpMode=RCIM_Cubic, tangentMode=RCTM_Break, time=1.7, value=1}
          {interpMode=RCIM_Cubic, leaveTangent=-5.0261, tangentMode=RCTM_Break, time=1.8, value=3}
          <at 2.5 1>
         ]}}}
      CurveIndex=$Engine.System.Age
     ))
    ScaleA=true
    ScaleRGB=true
   ) {module=ScaleColor001}
  ]
  rendererClasses [&NiagaraMeshRendererProperties]
  renderers [{rendererClass=&NiagaraMeshRendererProperties, rendererIndex=0}]
 }
}
summary {
 systemName=NS_Spire_Summon_Link
 emitters [
  {emitterName=Spline, rendererClasses [&NiagaraMeshRendererProperties]}
  {emitterName=Minimal004, rendererClasses [&NiagaraSpriteRendererProperties]}
 ]
 userVariables [
  User.SpikeAmount {type=NiagaraInt32, defaultValue=1}
  User.SpikePositionArray {
   type=NiagaraDataInterfaceArrayFloat3
   description=""
   defaultValue {
    struct=NiagaraExt_VariableValue_DataInterface
    value {
     dataInterface=&self:NiagaraDataInterfaceArrayFloat3_1
     dataInterfaceClass=&NiagaraDataInterfaceArrayFloat3
    }
   }
  }
  User.SpirePosition {type=Vector3f, description="", defaultValue {struct=Vector3f, value=<0 0 0>}}
 ]
}
